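; Gameplay cue notify for the Figment event Sleepy impact that plays impact audio on burst.
; OnBurst calls the parent, then spawns an impact 3D sound and a resonant variant at the cue location.
; It derives the resonant sound's integer "Size" parameter by subtracting the count of small, mid and large threshold range matches on raw magnitude from two.
; A leftover-style debug print fires on each burst, and OnExecute only forwards to the parent.
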(fn OnBurst
  (params (Target :type Actor%20Object%20Reference) (Parameters :type Gameplay%20Cue%20Parameters%20Structure%20%28by%20ref%29) (SpawnResults :type Gameplay%20Cue%20Notify%20Spawn%20Result%20Structure%20%28by%20ref%29))
  (bind (_normalizedmagnitude _rawmagnitude _effectcontext _matchedtagname _originaltag _aggregatedsourcetags _aggregatedtargettags _location) (Ability|GameplayCue|BreakGameplayCueParameters Parameters))
  (|CallParentFunction Target Parameters SpawnResults) ; Not in source GCN (from FN-8.5),\nbut included for safety
  ; Audio Modulation_2
  (bind _returnvalue (Audio|SpawnSoundatLocation 0 (Variables|Sound|GetImpactCue3D) _location))
  (bind _returnvalue_2 (Audio|SpawnSoundatLocation 0 (Variables|Sound|GetImpactCue3DResonant) _location))
  (Audio|Parameter|SetIntegerParameter _returnvalue_2 "Size" (- 2 (Math|Integer|int+int (Math|Conversions|ToInteger(Boolean) (Math|Float|InRange(Float) _rawmagnitude 0.0 (Variables|Default|GetSmallThreshold) true false)) (Math|Conversions|ToInteger(Boolean) (Math|Float|InRange(Float) _rawmagnitude 0.0 (Variables|Default|GetMidThreshold) false)) (Math|Conversions|ToInteger(Boolean) (Math|Float|InRange(Float) _rawmagnitude 0.0 (Variables|Default|GetLargeThreshold) false)))) :!class /Script/Engine.AudioComponent)
  (Development|PrintString 0 "OnBurst" true true "(R=1.000000,G=0.039546,B=0.029557,A=1.000000)" 1e+01)
  (return))

(fn OnExecute
  (params (MyTarget :type Actor%20Object%20Reference) (Parameters :type Gameplay%20Cue%20Parameters%20Structure%20%28by%20ref%29))
  (return (|CallParentFunction MyTarget Parameters)))
